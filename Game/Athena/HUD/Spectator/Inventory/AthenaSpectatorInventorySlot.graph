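# Widget Blueprint for one inventory slot in the Athena spectator HUD, built on AthenaSpectatorInventorySlotBase with a single native quick-bar cell as its root.
# The SetItem event stores the item and passes it to the cell to represent, and GetItem returns that stored item.
# The SetIsSelected event forwards selection to the cell, which has a zero-length selection animation and no selection offset.
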WidgetBlueprint /Game/Athena/HUD/Spectator/Inventory/AthenaSpectatorInventorySlot
graphs [
 EventGraph {nodes [
   n1 = Event SetItem!() -> (OutputDelegate, then, Item:Fort Item Object Reference) {is=title K2Node_Event}
   > n2 = Set Item(Item:Fort Item Object Reference=@n1.Item) -> (Output_Get:Fort Item Object Reference)
   > SetItemToRepresent!(
    execute
    self:Athena Quick Bar Native Cell Object Reference=Get AthenaQuickBarNativeCell() -> (AthenaQuickBarNativeCell:Athena Quick Bar Native Cell Object Reference)
    InItemToRepresent:Fort Item Object Reference=@n2
   ) -> (then)
   n3 = Event SetIsSelected!() -> (OutputDelegate, then, bIsSelected:bool) {is=title K2Node_Event}
   > SetIsSelected!(execute, self:Athena Quick Bar Native Cell Object Reference=Get AthenaQuickBarNativeCell(), bInSelected:bool=@n3.bIsSelected) -> (then)
  ]}
 GetItem {nodes [
   GetItem() {is=title K2Node_FunctionEntry}
   > Return Node!(execute, ReturnValue:Fort Item Object Reference=Get Item() -> (Item:Fort Item Object Reference))
  ]}
]
variables [Item {type=FortItem, default=null}]
widgets {
 info {
  inheritedWidgetCount=0
  parentClass=/Script/FortniteUI.AthenaSpectatorInventorySlotBase
  rootWidgetClass=/Script/FortniteUI.AthenaQuickBarNativeCell
  widgetCount=1
 }
 widgets [AthenaQuickBarNativeCell {
   widgetClassPath=&AthenaQuickBarNativeCell
   bIsVariable=true
   widget=AthenaQuickBarNativeCell
   properties {
    selectionAnimLength_Seconds=0
    visibility=SelfHitTestInvisible
    ammoTypeIcon_Padding=<4 8>
    ammoTypeIcon_Size=<18 18>
    emptyCell_Brush {
     bIsDynamicallyLoaded=false
     drawAs=Image
     imageType=NoImage
     mirroring=NoMirror
     resourceName=None
     resourceObject=/Game/UI/InGame/Quickbar/Quickbar-Frame-Default
     tiling=NoTile
     imageSize=<80 80>
     margin=<ltrb 0 0 0 0>
     outlineSettings {
      bUseBrushTransparency=false
      roundingType=HalfHeightRadius
      width=0
      color {colorUseRule=UseColor_Specified, specifiedColor=<rgba 0 0 0 0>}
      cornerRadii=<0 0 0 0>
     }
     tintColor {colorUseRule=UseColor_Specified, specifiedColor=<rgba 1 1 1 0.2>}
     uVRegion {bIsValid=false, max=<0 0>, min=<0 0>}
    }
    gadgetFuel_Brush {
     bIsDynamicallyLoaded=false
     drawAs=Image
     imageType=NoImage
     mirroring=NoMirror
     resourceName=None
     resourceObject=null
     tiling=NoTile
     imageSize=<6 32>
     margin=<ltrb 0 0 0 0>
     outlineSettings {
      bUseBrushTransparency=false
      roundingType=HalfHeightRadius
      width=0
      color {colorUseRule=UseColor_Specified, specifiedColor=<rgba 0 0 0 0>}
      cornerRadii=<0 0 0 0>
     }
     tintColor {colorUseRule=UseColor_Specified, specifiedColor=<rgba 1 1 1 1>}
     uVRegion {bIsValid=false, max=<0 0>, min=<0 0>}
    }
    gadgetFuel_Padding=<4 4>
    selectedCell_Brush {
     bIsDynamicallyLoaded=false
     drawAs=Image
     imageType=NoImage
     mirroring=NoMirror
     resourceName=None
     resourceObject=&Quickbar-Frame-Default
     tiling=NoTile
     imageSize=<24 24>
     margin=<ltrb 0 0 0 0>
     outlineSettings {
      bUseBrushTransparency=false
      roundingType=HalfHeightRadius
      width=0
      color {colorUseRule=UseColor_Specified, specifiedColor=<rgba 0 0 0 0>}
      cornerRadii=<0 0 0 0>
     }
     tintColor {colorUseRule=UseColor_Specified, specifiedColor=<rgba 1 1 1 1>}
     uVRegion {bIsValid=false, max=<0 0>, min=<0 0>}
    }
    selectionOffset=<0 0>
   }
  }]
}
